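# Physics asset for the hoodie mesh of the female medium Trey Cozy body cosmetic, apparently for secondary motion on the dyn_hoodie bone.
# Single capsules cover dyn_hoodie, pelvis, spine_05 and both thighs, forming a minimal body proxy rather than a full ragdoll.
# The hoodie is constrained to spine_05 with wide swing and very tight twist limits.
# Both thighs and spine_05 attach to the pelvis, with the spine's angular motion left free.
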PhysicsAsset /BRCosmetics/Characters/Player/Female/Medium/Bodies/F_MED_Trey_Cozy/Meshes/F_MED_Trey_Cozy_Hoodie_Physics
bodies {
 dyn_hoodie [{shapeType=Capsule, length=14.033, radius=3, shapeName=dyn_hoodie_capsule, center=<1.3 0 -0.5>, rotation=<rot 0 0 90>}]
 pelvis [{shapeType=Capsule, length=14, radius=7.25}]
 spine_05 [{shapeType=Capsule, length=16.9726, radius=5.9367, shapeName=spine_05_capsule, center=<7.9281 0.1475 0.3886>}]
 thigh_l [{shapeType=Capsule, length=15.75, radius=7.5, center=<-13.166 0.0001 0>, rotation=<rot 90 0 0>}]
 thigh_r [{shapeType=Capsule, length=15.75, radius=7.5, center=<13.166 0.0001 0>, rotation=<rot 90 0 0>}]
}
constraints [
 {bone1Name=thigh_l, bone2Name=pelvis, swing1LimitDegrees=45, swing2LimitDegrees=45, twistLimitDegrees=45}
 {bone1Name=thigh_r, bone2Name=pelvis, swing1LimitDegrees=45, swing2LimitDegrees=45, twistLimitDegrees=45}
 {
  bone1Name=spine_05
  bone2Name=pelvis
  swing1LimitDegrees=45
  swing1Motion=Free
  swing2LimitDegrees=45
  swing2Motion=Free
  twistLimitDegrees=45
  twistMotion=Free
 }
 {bone1Name=dyn_hoodie, bone2Name=spine_05, swing1LimitDegrees=55, swing2LimitDegrees=50, twistLimitDegrees=2}
]
