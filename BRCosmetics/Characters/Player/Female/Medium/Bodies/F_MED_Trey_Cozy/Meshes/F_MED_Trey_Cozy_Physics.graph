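# Physics asset for the F_MED_Trey_Cozy female player cosmetic body, covering two small dangling strings at the waist plus the pelvis and thighs.
# Each left and right string is a two-capsule chain constrained to the pelvis, with looser swing at the lower link and very tight twist throughout.
# Larger capsules on the pelvis and both thighs, with thighs constrained to the pelvis, provide body volume around the strings.
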PhysicsAsset /BRCosmetics/Characters/Player/Female/Medium/Bodies/F_MED_Trey_Cozy/Meshes/F_MED_Trey_Cozy_Physics
bodies {
 dyn_string_l_01 [{
   shapeType=Capsule
   length=2.505
   radius=0.505
   shapeName=dyn_string_l_01_capsule
   center=<1.0719 -0.0226 -0.0005>
   rotation=<rot 0 89.5628 -88.244>
  }]
 dyn_string_l_02 [{
   shapeType=Capsule
   length=2.5114
   radius=0.6
   shapeName=dyn_string_l_02_capsule
   center=<0.8426 -0.0212 -0.0602>
   rotation=<rot 0 -88.0952 -86.626>
  }]
 dyn_string_r_01 [{
   shapeType=Capsule
   length=2.505
   radius=0.505
   shapeName=dyn_string_r_01_capsule
   center=<1.0719 0.0229 -0.0021>
   rotation=<rot 0 90.5141 -88.2679>
  }]
 dyn_string_r_02 [{
   shapeType=Capsule
   length=2.5114
   radius=0.6
   shapeName=dyn_string_r_02_capsule
   center=<0.839 0.0964 -0.0264>
   rotation=<rot 0 -86.6402 -89.1183>
  }]
 pelvis [{shapeType=Capsule, length=14, radius=7.25}]
 thigh_l [{shapeType=Capsule, length=15.75, radius=7.5, center=<-13.166 0.0001 0>, rotation=<rot 90 0 0>}]
 thigh_r [{shapeType=Capsule, length=15.75, radius=7.5, center=<13.166 0.0001 0>, rotation=<rot 90 0 0>}]
}
constraints [
 {bone1Name=thigh_l, bone2Name=pelvis, swing1LimitDegrees=45, swing2LimitDegrees=45, twistLimitDegrees=45}
 {bone1Name=thigh_r, bone2Name=pelvis, swing1LimitDegrees=45, swing2LimitDegrees=45, twistLimitDegrees=45}
 {bone1Name=dyn_string_l_01, bone2Name=pelvis, swing1LimitDegrees=20, swing2LimitDegrees=20, twistLimitDegrees=3}
 {bone1Name=dyn_string_l_02, bone2Name=dyn_string_l_01, swing1LimitDegrees=40, swing2LimitDegrees=40, twistLimitDegrees=3}
 {bone1Name=dyn_string_r_01, bone2Name=pelvis, swing1LimitDegrees=20, swing2LimitDegrees=20, twistLimitDegrees=3}
 {bone1Name=dyn_string_r_02, bone2Name=dyn_string_r_01, swing1LimitDegrees=40, swing2LimitDegrees=40, twistLimitDegrees=3}
]
