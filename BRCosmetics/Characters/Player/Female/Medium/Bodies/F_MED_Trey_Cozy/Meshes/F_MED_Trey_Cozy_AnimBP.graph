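# Post-style Anim Blueprint for the F_MED_Trey_Cozy female body, adding secondary motion to a pose copied from the attached parent mesh.
# The pose runs through a Control Rig, Anim Dynamics shoestraps, then two RigidBody simulations using hoodie and body physics assets.
# EventGraph lowers the second simulation's front tassel alpha in the front end and derives neck and tilt alphas from head rotation.
# Hoodie dynamics, brim transforms and those head-driven alphas do not reach the output pose.
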AnimBlueprint /BRCosmetics/Characters/Player/Female/Medium/Bodies/F_MED_Trey_Cozy/Meshes/F_MED_Trey_Cozy_AnimBP
target_skeleton=&/F_MED_Trey_Cozy_Skeleton
graphs [
 AnimGraph {nodes [
   Anim Dynamics - Bone: dyn_hoodie(Alpha=0.75, ExternalForce="0,0,0") {
    angularDampingOverride=0.99
    bAngularSpring=true
    bGravityOverrideInSimSpace=true
    bOverrideAngularDamping=true
    bUseGravityOverride=true
    alphaBoolBlend {blendOption=Linear}
    boundBone {boneName=dyn_hoodie}
    componentAppliedLinearAccClamp=<1000 1000 1000>
    componentLinearAccScale=<0.3 0.3 0.3>
    componentLinearVelScale=<0.3 0.3 0.3>
    gravityOverride {z=-1000}
    physicsBodyDefinitions [{
      sphereCollisionRadius=0
      boundBone {boneName=dyn_hoodie}
      boxExtents=<7 10 7>
      constraintSetup {angularLimitsMax=<3 0 70>, angularLimitsMin=<-3 -90 -70>, angularTarget=<0 0 0>}
      localJointOffset=<0.6 0 -0.2>
     }]
    relativeSpaceBone {boneName=spine_05}
   }
   n1 = "Transform (Modify) Bone - Bone: dyn_hoddiebrim"(Alpha=1, Translation="0.000,-0.976,-1", Rotation="0.000,-0.000,-15", Scale="1,1,1") {
    rotationMode=BMM_Additive
    rotationSpace=BCS_BoneSpace
    translationMode=BMM_Additive
    translationSpace=BCS_BoneSpace
    alphaBoolBlend {blendOption=Linear}
    boneToModify {boneName=dyn_hoddiebrim}
    rotation {roll=-15}
    translation {y=-0.976, z=-1}
   }
   "Transform (Modify) Bone - Bone: dyn_hoddiebrim"(
    ComponentPose=@n1
    Alpha=Get Neck Alpha() -> (Neck Alpha:double)
    Translation="-0.000,-2,-6"
    Rotation="0.000,-0.000,-25"
    Scale="0.500,0.750,0.750"
   ) {
    rotationMode=BMM_Additive
    rotationSpace=BCS_BoneSpace
    scaleMode=BMM_Additive
    scaleSpace=BCS_BoneSpace
    translationMode=BMM_Additive
    translationSpace=BCS_BoneSpace
    alphaBoolBlend {blendOption=Linear}
    boneToModify {boneName=dyn_hoddiebrim}
    rotation {roll=-25}
    scale=<0.5 0.75 0.75>
    translation {y=-2, z=-6}
   }
   n2 = Control Rig(Source=Copy Pose From Mesh() {bUseAttachedParent=true}) {
    alphaBoolBlend {blendOption=Linear}
    controlRigAssetReference {blueprintRigClass=&/F_MED_Trey_Cozy_CtrlRig_sculpters.F_MED_Trey_Cozy_CtrlRig_sculpters_C}
   }
   n3 = Anim Dynamics - Bone: dyn_shoestrap_l(ComponentPose=Local To Component(LocalPose=@n2), Alpha=0.3, ExternalForce="0,0,0") {
    angularDampingOverride=0.99
    angularSpringConstant=100
    bAngularSpring=true
    bGravityOverrideInSimSpace=true
    bOverrideAngularDamping=true
    bUseGravityOverride=true
    simulationSpace=BoneRelative
    alphaBoolBlend {blendOption=Linear}
    boundBone {boneName=dyn_shoestrap_l}
    componentAppliedLinearAccClamp=<100 100 100>
    componentLinearAccScale=<0.2 0.2 0.2>
    componentLinearVelScale=<0.2 0.2 0.2>
    gravityOverride {z=-1000}
    physicsBodyDefinitions [{
      sphereCollisionRadius=0
      boundBone {boneName=dyn_shoestrap_l}
      boxExtents=<6 6 7>
      constraintSetup {angularLimitsMax=<0 0 50>, angularLimitsMin=<0 0 -20>, angularTarget=<0 0 0>}
      localJointOffset=<1 0 0>
     }]
    relativeSpaceBone {boneName=root}
   }
   n4 = Anim Dynamics - Bone: dyn_shoestrap_r(ComponentPose=@n3, Alpha=0.3, ExternalForce="0,0,0") {
    angularDampingOverride=0.99
    angularSpringConstant=100
    bAngularSpring=true
    bGravityOverrideInSimSpace=true
    bOverrideAngularDamping=true
    bUseGravityOverride=true
    simulationSpace=BoneRelative
    alphaBoolBlend {blendOption=Linear}
    boundBone {boneName=dyn_shoestrap_r}
    componentAppliedLinearAccClamp=<100 100 100>
    componentLinearAccScale=<0.2 0.2 0.2>
    componentLinearVelScale=<0.2 0.2 0.2>
    gravityOverride {z=-1000}
    physicsBodyDefinitions [{
      sphereCollisionRadius=0
      boundBone {boneName=dyn_shoestrap_r}
      boxExtents=<6 6 7>
      constraintSetup {angularLimitsMax=<0 0 50>, angularLimitsMin=<0 0 -20>, angularTarget=<0 0 0>}
      localJointOffset=<1 0 0>
     }]
    relativeSpaceBone {boneName=root}
   }
   n5 = RigidBody(ComponentPose=@n4, Alpha=0.5) {
    bOverrideWorldGravity=true
    overridePhysicsAsset=&/F_MED_Trey_Cozy_Hoodie_Physics
    simulationSpace=BaseBoneSpace
    alphaBoolBlend {blendOption=Linear}
    baseBoneRef {boneName=spine_05}
    componentAppliedLinearAccClamp=<5000 5000 5000>
    componentLinearAccScale=<0.2 0.4 0.5>
    componentLinearVelScale=<0.2 0.4 0.5>
    overrideWorldGravity {z=-980}
   }
   n6 = RigidBody(ComponentPose=@n5, Alpha=Get Front tassel Alpha() -> (Front tassel Alpha:double)) {
    bOverrideWorldGravity=true
    overridePhysicsAsset=&/F_MED_Trey_Cozy_Physics
    simulationSpace=BaseBoneSpace
    alphaBoolBlend {blendOption=Linear}
    baseBoneRef {boneName=pelvis}
    componentAppliedLinearAccClamp=<3000 3000 3000>
    componentLinearAccScale=<0.2 0.2 0.2>
    componentLinearVelScale=<0.2 0.2 0.2>
    overrideWorldGravity {z=-500}
   }
   Output Pose(Result=Component To Local(ComponentPose=@n6)) {layerGroup=None}
   Get Head Tilt Alpha() -> (Head Tilt Alpha:double)
  ]}
 EventGraph {nodes [
   Event BlueprintUpdateAnimation()
   n1 = BreakTransform(InTransform=GetDeltaTransformFromRefPose(self=GetOwnerSkeletalMeshComponent(), BoneName=head, BaseName=spine_05))
   > Set Neck Alpha(Neck Alpha:double=MapRangeClamped(Value=BreakRotator(InRot=@n1.Rotation).Yaw, InRangeA=-10, InRangeB=-40)) -> (Output_Get:double)
   n2 = BreakTransform(InTransform=GetDeltaTransformFromRefPose(self=GetOwnerSkeletalMeshComponent(), BoneName=head, BaseName=spine_05))
   > Set Head Tilt Alpha(Head Tilt Alpha:double=MapRangeClamped(Value=BreakRotator(InRot=@n2.Rotation).Pitch, InRangeA=0, InRangeB=30)) -> (Output_Get:double)
   > Set Front tassel Alpha(Front tassel Alpha:double=SelectFloat(A=0.25, B=1, bPickA=Get bIsInFrontEnd())) -> (Output_Get:double)
  ]}
]
variables [
 Front tassel Alpha {type=number, default=1}
 Head Tilt Alpha {type=number}
 Neck Alpha {type=number}
]
